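# Translucent UI material that, by its name, serves as a background template for the vehicle mods UI.
# An exponential radial gradient blends two opaque blue-grey colors, one dark and one lighter.
# The Outline output of MF_UI_Box then blends that gradient toward white with zero alpha, using a thin outline thickness.
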Material /VehicleModsUI/Art/Materials/M_UI_VehicleMods_BackTemplate
graph {
 properties {
  bAutomaticallySetUsageInEditor=false
  blendMode=BLEND_Translucent
  materialDomain=MD_UI
  pixelDepthOffsetMode=PDOM_Legacy
  usedShadingModels=Default Lit
 }
 nodes [
  n1 = Constant() {r=0.6}
  LinearInterpolate(
   A=LinearInterpolate(
    A=Constant4Vector() {constant=<rgba 0.1356 0.1441 0.1559 1>}.RGBA
    B=Constant4Vector() {constant=<rgba 0.3864 0.4179 0.4508 1>}.RGBA
    Alpha=MaterialFunctionCall(Radius=@n1) {materialFunction=&RadialGradientExponential}.RadialGradientExponential
   )
   B=Constant4Vector() {constant=<rgba 1 1 1 0>}.RGBA
   Alpha=MaterialFunctionCall(Size=Constant() {r=1}, OutlineThickness=Constant() {r=0.05}) {materialFunction=&MF_UI_Box}.Outline
  )
 ]
}
