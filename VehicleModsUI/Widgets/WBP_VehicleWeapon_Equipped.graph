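# HUD widget showing the equipped vehicle-mod weapon's icon, beside gun and swap icons and a turret-equip keybind prompt.
# On Construct it binds to the HUD context's pawn-set event, which then binds weapon-equipped and refreshes immediately.
# Refreshes read the cached current weapon; SetWeaponIcon maps its exact class to a machine gun or mortar texture.
# Other weapons leave the icon unchanged.
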WidgetBlueprint /VehicleModsUI/Widgets/WBP_VehicleWeapon_Equipped
graphs [
 EventGraph {nodes [
   n1 = On Weapon Equipped!() -> (OutputDelegate, then, NewWeapon:Fort Weapon Object Reference, PrevWeapon:Fort Weapon Object Reference) {is=title K2Node_CustomEvent}
   n2 = Cast To FortWeaponRangedForVehicleMod!(
    execute=[@n4.then, @n1.then]
    Object=GetCachedCurrentWeapon!(self:Fort HUDContext Object Reference=Get FortHUDContext!() -> (ReturnValue:Fort HUDContext Object Reference) {is=title K2Node_GetBlueprintContextEx}) -> (ReturnValue:Fort Weapon Object Reference)
   ) -> (then, CastFailed, AsFort Weapon Ranged for Vehicle Mod:Fort Weapon Ranged For Vehicle Mod Object Reference) {is=title K2Node_DynamicCast}
   > SetWeaponIcon!(execute, self, Weapon:Object Reference=@n2) -> (then)
   n3 = OnPawnSet() {is=title K2Node_CustomEvent}
   n4 = Bind Event to On Weapon Equipped Delegate(execute=[@n5.then, @n3.then], self:Fort HUDContext Object Reference=Get FortHUDContext(), Delegate=@n1) {is=title K2Node_AddDelegate}
   Event Construct()
   > n5 = Bind Event to On Pawn Set(self:Fort HUDContext Object Reference=Get FortHUDContext(), Delegate=@n3) {is=title K2Node_AddDelegate}
  ]}
 SetWeaponIcon {nodes [
   n1 = SetWeaponIcon!() -> (then, Weapon:Object Reference) {is=title K2Node_FunctionEntry}
   > n2 = Cast To FortWeaponRangedForVehicleMod Class!(execute:Exec, Class:Object Class Reference=Get Class(Object=@n1) -> (ReturnValue:Object Class Reference)) -> (then:Exec, CastFailed:Exec, AsFort Weapon Ranged for Vehicle Mod:Fort Weapon Ranged For Vehicle Mod Class Reference) {is=title K2Node_ClassDynamicCast}
   n3 = Find(
    TargetMap:Map of Fort Weapon Ranged For Vehicle Mod Class References to Texture Object References=Get WeaponToIconMap() -> (WeaponToIconMap:Map of Fort Weapon Ranged For Vehicle Mod Class References to Texture Object References)
    Key:"Fort Weapon Ranged For Vehicle Mod Class Reference (by ref)"=@n2
   ) -> (Value:Texture Object Reference)
   > Branch(Condition=@n3.ReturnValue)
   > n4 = Cast To Texture2D(Object=@n3)
   > SetBrushFromTexture!(
    execute
    self:Image Object Reference=Get Image_Weapon() -> (Image_Weapon:Image Object Reference)
    Texture:Texture 2D Object Reference=@n4
    bMatchSize:bool=false
   ) -> (then)
  ]}
]
variables [
 Minimized {type=boolean}
 WeaponToIconMap {default {
   /Script/Engine.BlueprintGeneratedClass'/VehicleMods/Items/MachineGunMod/B_VehicleMachineGunModWeapon.B_VehicleMachineGunModWeapon_C'=/VehicleModsUI/Art/Textures/MachineGun/T_UI_MachineGun
   &BlueprintGeneratedClass'/VehicleMods/Items/MortarGunMod/B_VehicleMortarWeapon.B_VehicleMortarWeapon_C'=/VehicleModsUI/Art/Textures/Mortar/T_UI_Mortar
  }}
]
widgets {
 info {inheritedWidgetCount=0, parentClass=&UserWidget, rootWidgetClass=&HorizontalBox, widgetCount=9}
 widgets [
  HorizontalBox_38 {widgetClassPath=&HorizontalBox, widget=HorizontalBox_38}
  SizeBox_1 {
   widgetClassPath=&SizeBox
   parent=HorizontalBox_38
   widget=SizeBox_1
   properties {bOverride_HeightOverride=true, bOverride_WidthOverride=true, heightOverride=75, widthOverride=75}
   slot {class=HorizontalBoxSlot, properties {horizontalAlignment=HAlign_Center, verticalAlignment=VAlign_Top, padding=<ltrb 0 10 0 0>}}
  }
  Border_Weapon {
   widgetClassPath=&Border
   bIsVariable=true
   parent=SizeBox_1
   widget=Border_Weapon
   properties {
    horizontalAlignment=HAlign_Center
    verticalAlignment=VAlign_Center
    background {
     bIsDynamicallyLoaded=false
     drawAs=Image
     imageType=FullColor
     mirroring=NoMirror
     resourceName=None
     resourceObject=/VehicleModsUI/Art/Materials/M_UI_VehicleMods_BackTemplate
     tiling=NoTile
     imageSize=<32 32>
     margin=<ltrb 0 0 0 0>
     outlineSettings {
      bUseBrushTransparency=false
      roundingType=HalfHeightRadius
      width=0
      color {colorUseRule=UseColor_Specified, specifiedColor=<rgba 0 0 0 0>}
      cornerRadii=<0 0 0 0>
     }
     tintColor {colorUseRule=UseColor_Specified, specifiedColor=<rgba 1 1 1 1>}
     uVRegion {bIsValid=false, max=<0 0>, min=<0 0>}
    }
   }
   slot {class=SizeBoxSlot, properties {horizontalAlignment=HAlign_Center, verticalAlignment=VAlign_Center}}
  }
  Image_Weapon {
   widgetClassPath=&Image
   bIsVariable=true
   parent=Border_Weapon
   widget=Image_Weapon
   properties {brush {
     bIsDynamicallyLoaded=false
     drawAs=Image
     imageType=FullColor
     mirroring=NoMirror
     resourceName=None
     resourceObject=&T_UI_MachineGun
     tiling=NoTile
     imageSize=<150 150>
     margin=<ltrb 0 0 0 0>
     outlineSettings {
      bUseBrushTransparency=false
      roundingType=HalfHeightRadius
      width=0
      color {colorUseRule=UseColor_Specified, specifiedColor=<rgba 0 0 0 0>}
      cornerRadii=<0 0 0 0>
     }
     tintColor {colorUseRule=UseColor_Specified, specifiedColor=<rgba 1 1 1 1>}
     uVRegion {bIsValid=false, max=<0 0>, min=<0 0>}
    }}
   slot {class=BorderSlot, properties {horizontalAlignment=HAlign_Center, verticalAlignment=VAlign_Center}}
  }
  Spacer_210 {widgetClassPath=&Spacer, parent=HorizontalBox_38, widget=Spacer_210, properties {size=<15 1>}, slot {class=HorizontalBoxSlot}}
  VerticalBox_0 {
   widgetClassPath=&VerticalBox
   parent=HorizontalBox_38
   widget=VerticalBox_0
   slot {class=HorizontalBoxSlot, properties {padding=<ltrb 0 10 20 0>}}
  }
  Image {
   widgetClassPath=&Image
   bIsVariable=true
   parent=VerticalBox_0
   widget=Image
   properties {brush {
     bIsDynamicallyLoaded=false
     drawAs=Image
     imageType=FullColor
     mirroring=NoMirror
     resourceName=None
     resourceObject=/VehicleModsUI/Art/Textures/T_UI_VehicleModsIcon_64_Guns
     tiling=NoTile
     imageSize=<45 45>
     margin=<ltrb 0 0 0 0>
     outlineSettings {
      bUseBrushTransparency=false
      roundingType=HalfHeightRadius
      width=0
      color {colorUseRule=UseColor_Specified, specifiedColor=<rgba 0 0 0 0>}
      cornerRadii=<0 0 0 0>
     }
     tintColor {colorUseRule=UseColor_Specified, specifiedColor=<rgba 1 1 1 1>}
     uVRegion {bIsValid=false, max=<0 0>, min=<0 0>}
    }}
   slot {class=VerticalBoxSlot, properties {padding=<ltrb 0 15 0 0>}}
  }
  Image_67 {
   widgetClassPath=&Image
   bIsVariable=true
   parent=VerticalBox_0
   widget=Image_67
   properties {brush {
     bIsDynamicallyLoaded=false
     drawAs=Image
     imageType=FullColor
     mirroring=NoMirror
     resourceName=None
     resourceObject=/Game/Athena/UI/InGame/Creative/PropertyEditor/Icon_Picker/Textures_V2/64x64/T_UI_IconLibrary_Swap_64
     tiling=NoTile
     imageSize=<25 25>
     margin=<ltrb 0 0 0 0>
     outlineSettings {
      bUseBrushTransparency=false
      roundingType=HalfHeightRadius
      width=0
      color {colorUseRule=UseColor_Specified, specifiedColor=<rgba 0 0 0 0>}
      cornerRadii=<0 0 0 0>
     }
     tintColor {colorUseRule=UseColor_Specified, specifiedColor=<rgba 1 1 1 1>}
     uVRegion {bIsValid=false, max=<0 0>, min=<0 0>}
    }}
   slot {class=VerticalBoxSlot, properties {horizontalAlignment=HAlign_Center, verticalAlignment=VAlign_Center}}
  }
  KeybindWidget {
   widgetClassPath=/FortUILibrary/Input/KeybindWidget
   bIsVariable=true
   parent=VerticalBox_0
   widget=KeybindWidget
   properties {boundInputAction=/VehicleMods/Input/IA_VehicleWeaponSeatMod_TurretEquip}
   slot {class=VerticalBoxSlot, properties {horizontalAlignment=HAlign_Center, verticalAlignment=VAlign_Center, padding=<ltrb 0 2 0 0>}}
  }
 ]
}
