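# HUD widget for an unequipped vehicle mod weapon, showing the weapon icon, a swap icon and the turret-equip keybind.
# On construct and pawn change, it reads the player's seat weapon component and maps the cached definition to a machine gun or mortar icon.
# If enabled, it hides pickaxe slot keybinds for mouse and keyboard, following input changes and restoring them on destruct.
# Anim_In adds a slide, fade, glow and sparkles; Anim_Out fades it out.
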WidgetBlueprint /VehicleModsUI/Widgets/WBP_VehicleWeapon_Unequipped
animations [
 Anim_In {
  binding_count=4
  length_seconds=1.4
  object=&self:Anim_In
  sequence {bindings [
    Fx_Sparkle_02 {
     id=#1
     possessed_class=Image
     tracks [{
       type=MovieSceneFloatTrack
       display_name=Render Opacity
       sections [{
         type=MovieSceneFloatSection
         end_frame=28
         start_frame=0
         channels [None {type=MovieSceneScriptingFloatChannel, keys [[0,0],[10,0],[15,1],[23,1],[28,0]]}]
        }]
      }]
    }
    Fx_Glow {
     id=#2
     possessed_class=Image
     tracks [
      {
       type=MovieSceneFloatTrack
       display_name=Render Opacity
       sections [{
         type=MovieSceneFloatSection
         end_frame=19
         start_frame=0
         channels [None {type=MovieSceneScriptingFloatChannel, keys [[0,0],[7.05,0],[13,0.3],[19.751,0]]}]
        }]
      }
      {
       type=MovieScene2DTransformTrack
       sections [{
         type=MovieScene2DTransformSection
         end_frame=19
         start_frame=0
         channels [
          Scale.X {type=MovieSceneScriptingFloatChannel, keys [[0,1],[7.05,0.5],[13,1.2],[19.751,2.3]]}
          Scale.Y {type=MovieSceneScriptingFloatChannel, keys [[0,1],[7.05,0.5],[13,1.2],[19.751,2.3]]}
         ]
        }]
      }
     ]
    }
    Fx_Sparkle {
     id=#3
     possessed_class=Image
     tracks [{
       type=MovieSceneFloatTrack
       display_name=Render Opacity
       sections [{
         type=MovieSceneFloatSection
         end_frame=28
         start_frame=0
         channels [None {type=MovieSceneScriptingFloatChannel, keys [[0,0],[10,0],[15,1],[23,1],[28,0]]}]
        }]
      }]
    }
    HorizontalBox_38 {
     id=#4
     possessed_class=HorizontalBox
     tracks [
      {
       type=MovieSceneFloatTrack
       display_name=Render Opacity
       sections [{type=MovieSceneFloatSection, end_frame=10, start_frame=0, channels [None {type=MovieSceneScriptingFloatChannel, keys [[0,0],[10,1]]}]}]
      }
      {
       type=MovieScene2DTransformTrack
       sections [{
         type=MovieScene2DTransformSection
         end_frame=10
         start_frame=0
         channels [
          Translation.X {type=MovieSceneScriptingFloatChannel, keys [[0,15],[10,0]]}
          Translation.Y {type=MovieSceneScriptingFloatChannel, const [0,1]}
         ]
        }]
      }
     ]
    }
   ]}
 }
 Anim_Out {
  binding_count=1
  length_seconds=0.5
  object=&self:Anim_Out
  sequence {bindings [HorizontalBox_38 {
     id=#5
     possessed_class=HorizontalBox
     tracks [{
       type=MovieSceneFloatTrack
       display_name=Render Opacity
       sections [{
         type=MovieSceneFloatSection
         end_frame=10
         start_frame=0
         channels [None {type=MovieSceneScriptingFloatChannel, keys [[0,1],[5,1],[10.0003,0]]}]
        }]
      }]
    }]}
 }
]
graphs [
 AdjustPickaxeKeybindVisibility {nodes [
   n1 = AdjustPickaxeKeybindVisibility!() -> (then, Input Type:ECommonInputType Enum) {is=title K2Node_FunctionEntry}
   > Branch(Condition=Get bShouldRequestHidePickaxeSlotKeybinds() -> (bShouldRequestHidePickaxeSlotKeybinds:bool))
   n2 = Get FortHUDContext!() -> (ReturnValue:Fort HUDContext Object Reference) {is=title K2Node_GetBlueprintContextEx}
   > Is Valid(InputObject=@n2)
   > RequestHidePickaxeSlotKeybinds!(execute, self:Fort HUDContext Object Reference=@n2, bShouldHide:bool="Equal (Enum)"(A=@n1, B=MouseAndKeyboard)) -> (then)
  ]}
 EventGraph {nodes [
   n1 = OnPawnSet() {is=title K2Node_CustomEvent}
   n2 = Cast To FortPlayerPawnAthena(execute=[@n12.then, @n1.then], Object=GetOwningPlayerPawn!(self:User Widget Object Reference) -> (ReturnValue:Pawn Object Reference))
   > n3 = Cast To FortVehicleInterface!(
    execute
    Object=Get Vehicle!(self:Fort Player Pawn Object Reference=@n2) -> (ReturnValue:Fort Athena Vehicle Object Reference) {is=title K2Node_CallFunction}
   ) -> (then, CastFailed, AsFort Vehicle Interface:Fort Vehicle Interface Interface) {is=title K2Node_DynamicCast}
   n4 = GetVehicleSeatIndex!(self:Fort Player Pawn Object Reference=@n2) -> (ReturnValue:int=0)
   > SetWeaponIcon!(
    execute
    self
    Weapon:Fort Weapon Ranged Item Definition Object Reference=Get CachedWeaponDef!(self:Fort Vehicle Seat Weapon Component Object Reference=GetSeatWeaponComponent!(self:Fort Vehicle Interface Interface=@n3, SeatIndex:int=@n4) -> (ReturnValue:Fort Vehicle Seat Weapon Component Object Reference)) -> (CachedWeaponDef:Fort Weapon Ranged Item Definition Object Reference)
   ) -> (then)
   Event Destruct()
   n5 = Get CommonInputSubsystem()
   > n6 = Is Valid(InputObject=@n5)
   > n7 = Unbind Event from On Input Method Changed(self=@n5, Delegate=Create Event())
   n8 = Get FortHUDContext()
   Is Valid(exec=[@n7.then, @n6.Is Not Valid], InputObject=@n8)
   > Branch(Condition=Get bShouldRequestHidePickaxeSlotKeybinds())
   > RequestHidePickaxeSlotKeybinds(self=@n8, bShouldHide=false)
   Event Construct()
   n9 = Get CommonInputSubsystem()
   > n10 = Is Valid(InputObject=@n9)
   > On New Input Type!(execute, self, NewInputType:ECommonInputType Enum=GetCurrentInputType(self=Get CommonInputSubsystem())) -> (then)
   > n11 = Bind Event to On Input Method Changed(self=@n9, Delegate=Create Event())
   n12 = Bind Event to On Pawn Set(execute=[@n11.then, @n10.Is Not Valid], self:Fort HUDContext Object Reference=Get FortHUDContext(), Delegate=@n1) {is=title K2Node_AddDelegate}
  ]}
 On New Input Type {nodes [
   n1 = On New Input Type!() -> (then, NewInputType:ECommonInputType Enum) {is=title K2Node_FunctionEntry}
   > Branch(Condition=Get bShouldRequestHidePickaxeSlotKeybinds())
   > AdjustPickaxeKeybindVisibility!(execute, self, Input Type:ECommonInputType Enum=@n1) -> (then) {is=title K2Node_CallFunction}
  ]}
 SetWeaponIcon {nodes [
   n1 = SetWeaponIcon!() -> (then, Weapon:Fort Weapon Ranged Item Definition Object Reference) {is=title K2Node_FunctionEntry}
   n2 = Find(
    TargetMap:Map of Fort Weapon Ranged Item Definition Object References to Texture Object References=Get WeaponDefinitionToIconMap() -> (WeaponDefinitionToIconMap:Map of Fort Weapon Ranged Item Definition Object References to Texture Object References)
    Key:"Fort Weapon Ranged Item Definition Object Reference (by ref)"=@n1
   ) -> (Value:Texture Object Reference)
   > Branch(Condition=@n2.ReturnValue)
   > n3 = Cast To Texture2D(Object=@n2)
   > SetBrushFromTexture!(
    execute
    self:Image Object Reference=Get Image_Weapon() -> (Image_Weapon:Image Object Reference)
    Texture:Texture 2D Object Reference=@n3
    bMatchSize:bool=false
   ) -> (then)
  ]}
]
variables [
 bShouldRequestHidePickaxeSlotKeybinds {type=boolean, default=true}
 Minimized {type=boolean}
 WeaponDefinitionToIconMap {default {
   /Script/FortniteGame.FortWeaponRangedItemDefinition'/VehicleMods/Items/MachineGunMod/WID_VehicleMachineGunMod.WID_VehicleMachineGunMod'=/VehicleMods/UI/MachineGun/Icons/T_VehicleModsIcon_MachineGun_128
   &FortWeaponRangedItemDefinition'/VehicleMods/Items/MortarGunMod/WID_VehicleMortarMod.WID_VehicleMortarMod'=/VehicleMods/UI/Mortar/Icons/T_VehicleModsIcon_Mortar_128
  }}
]
widgets {
 info {inheritedWidgetCount=0, parentClass=&UserWidget, rootWidgetClass=&HorizontalBox, widgetCount=10}
 widgets [
  HorizontalBox_38 {widgetClassPath=&HorizontalBox, widget=HorizontalBox_38}
  Spacer_210 {widgetClassPath=&Spacer, parent=HorizontalBox_38, widget=Spacer_210, properties {size=<15 1>}, slot {class=HorizontalBoxSlot}}
  VerticalBox_0 {widgetClassPath=&VerticalBox, parent=HorizontalBox_38, widget=VerticalBox_0, slot {class=HorizontalBoxSlot}}
  Overlay_125 {widgetClassPath=&Overlay, parent=VerticalBox_0, widget=Overlay_125, slot {class=VerticalBoxSlot}}
  Image_Weapon {
   widgetClassPath=&Image
   bIsVariable=true
   parent=Overlay_125
   widget=Image_Weapon
   properties {brush {
     bIsDynamicallyLoaded=false
     drawAs=Image
     imageType=FullColor
     mirroring=NoMirror
     resourceName=None
     resourceObject=/VehicleMods/UI/MachineGun/Icons/T_VehicleModsIcon_MachineGun_64
     tiling=NoTile
     imageSize=<65 65>
     margin=<ltrb 0 0 0 0>
     outlineSettings {
      bUseBrushTransparency=false
      roundingType=HalfHeightRadius
      width=0
      color {colorUseRule=UseColor_Specified, specifiedColor=<rgba 0 0 0 0>}
      cornerRadii=<0 0 0 0>
     }
     tintColor {colorUseRule=UseColor_Specified, specifiedColor=<rgba 1 1 1 1>}
     uVRegion {bIsValid=false, max=<0 0>, min=<0 0>}
    }}
   slot {class=OverlaySlot, properties {horizontalAlignment=HAlign_Fill, verticalAlignment=VAlign_Fill, padding=<ltrb 0 10 0 0>}}
  }
  Fx_Glow {
   widgetClassPath=&Image
   parent=Overlay_125
   widget=Fx_Glow
   properties {
    renderOpacity=0
    brush {
     bIsDynamicallyLoaded=false
     drawAs=Image
     imageType=FullColor
     mirroring=NoMirror
     resourceName=None
     resourceObject=/Game/Athena/HUD/LTM/Wax/Materials/M_UI_TextGlow_Add
     tiling=NoTile
     imageSize=<50 50>
     margin=<ltrb 0 0 0 0>
     outlineSettings {
      bUseBrushTransparency=false
      roundingType=HalfHeightRadius
      width=0
      color {colorUseRule=UseColor_Specified, specifiedColor=<rgba 0 0 0 0>}
      cornerRadii=<0 0 0 0>
     }
     tintColor {colorUseRule=UseColor_Specified, specifiedColor=<rgba 1 0.8944 0.3281 1>}
     uVRegion {bIsValid=false, max=<0 0>, min=<0 0>}
    }
   }
   slot {class=OverlaySlot, properties {horizontalAlignment=HAlign_Center, verticalAlignment=VAlign_Center}}
  }
  Fx_Sparkle {
   widgetClassPath=&Image
   parent=Overlay_125
   widget=Fx_Sparkle
   properties {
    renderOpacity=0
    brush {
     bIsDynamicallyLoaded=false
     drawAs=Image
     imageType=FullColor
     mirroring=NoMirror
     resourceName=None
     resourceObject=/Game/Athena/HUD/Materials/Victory/M_Flare_Cross_Animated
     tiling=NoTile
     imageSize=<20 20>
     margin=<ltrb 0 0 0 0>
     outlineSettings {
      bUseBrushTransparency=false
      roundingType=HalfHeightRadius
      width=0
      color {colorUseRule=UseColor_Specified, specifiedColor=<rgba 0 0 0 0>}
      cornerRadii=<0 0 0 0>
     }
     tintColor {colorUseRule=UseColor_Specified, specifiedColor=<rgba 1 1 1 1>}
     uVRegion {bIsValid=false, max=<0 0>, min=<0 0>}
    }
   }
   slot {class=OverlaySlot, properties {horizontalAlignment=HAlign_Center, verticalAlignment=VAlign_Center, padding=<ltrb 0 -20 -15 0>}}
  }
  Fx_Sparkle_02 {
   widgetClassPath=&Image
   parent=Overlay_125
   widget=Fx_Sparkle_02
   properties {
    renderOpacity=0
    brush {
     bIsDynamicallyLoaded=false
     drawAs=Image
     imageType=FullColor
     mirroring=NoMirror
     resourceName=None
     resourceObject=&M_Flare_Cross_Animated
     tiling=NoTile
     imageSize=<20 20>
     margin=<ltrb 0 0 0 0>
     outlineSettings {
      bUseBrushTransparency=false
      roundingType=HalfHeightRadius
      width=0
      color {colorUseRule=UseColor_Specified, specifiedColor=<rgba 0 0 0 0>}
      cornerRadii=<0 0 0 0>
     }
     tintColor {colorUseRule=UseColor_Specified, specifiedColor=<rgba 1 1 1 1>}
     uVRegion {bIsValid=false, max=<0 0>, min=<0 0>}
    }
   }
   slot {class=OverlaySlot, properties {horizontalAlignment=HAlign_Center, verticalAlignment=VAlign_Center, padding=<ltrb -30 0 0 -30>}}
  }
  Image_67 {
   widgetClassPath=&Image
   bIsVariable=true
   parent=VerticalBox_0
   widget=Image_67
   properties {brush {
     bIsDynamicallyLoaded=false
     drawAs=Image
     imageType=FullColor
     mirroring=NoMirror
     resourceName=None
     resourceObject=/Game/Athena/UI/InGame/Creative/PropertyEditor/Icon_Picker/Textures_V2/64x64/T_UI_IconLibrary_Swap_64
     tiling=NoTile
     imageSize=<25 25>
     margin=<ltrb 0 0 0 0>
     outlineSettings {
      bUseBrushTransparency=false
      roundingType=HalfHeightRadius
      width=0
      color {colorUseRule=UseColor_Specified, specifiedColor=<rgba 0 0 0 0>}
      cornerRadii=<0 0 0 0>
     }
     tintColor {colorUseRule=UseColor_Specified, specifiedColor=<rgba 1 1 1 1>}
     uVRegion {bIsValid=false, max=<0 0>, min=<0 0>}
    }}
   slot {class=VerticalBoxSlot, properties {horizontalAlignment=HAlign_Center, verticalAlignment=VAlign_Center, padding=<ltrb 0 -5 0 0>}}
  }
  KeybindWidget {
   widgetClassPath=/FortUILibrary/Input/KeybindWidget
   bIsVariable=true
   parent=VerticalBox_0
   widget=KeybindWidget
   properties {boundInputAction=/VehicleMods/Input/IA_VehicleWeaponSeatMod_TurretEquip}
   slot {class=VerticalBoxSlot, properties {horizontalAlignment=HAlign_Center, verticalAlignment=VAlign_Center, padding=<ltrb 0 2 0 0>}}
  }
 ]
}
