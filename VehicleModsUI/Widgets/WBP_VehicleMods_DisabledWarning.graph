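# Reticle status widget that warns the local player when the vehicle they occupy is disabled or its driver is on another team.
# It follows the player controller's view target pawn and rebinds to that pawn's vehicle for disabled changes and seat entry or exit.
# Those events, with seat changes limited to the driver seat, refresh visibility and pick the localized "VehicleDisabled" or "EnemyPlayerOwned" text, hiding otherwise.
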WidgetBlueprint /VehicleModsUI/Widgets/WBP_VehicleMods_DisabledWarning
graphs [
 EventGraph {nodes [
   Event Construct()
   > Parent: Construct()
   > n1 = Cast To FortPlayerController(Object=GetOwningPlayer())
   n2 = Create Event()
   > Bind Event to On View Target Changed(self:Fort Player Controller Object Reference=@n1, Delegate=@n2) {is=title K2Node_AddDelegate}
   > SetViewTarget!(
    execute
    self
    NewViewTarget:Actor Object Reference=GetViewTarget!(self:Controller Object Reference=@n1) -> (ReturnValue:Actor Object Reference)
   ) -> (then)
   RefreshWarningState(execute=@n1.CastFailed)
   Event Destruct()
   > Parent: Destruct()
   > SetViewTarget()
   > n3 = Cast To FortPlayerController(Object=GetOwningPlayer())
   > Unbind Event from On View Target Changed(self:Fort Player Controller Object Reference=@n3, Delegate=@n2) {is=title K2Node_RemoveDelegate}
  ]}
 OnPawnEnteredExitedSeat {nodes [
   n1 = OnPawnEnteredExitedSeat!() -> (then, Vehicle:"Fort Vehicle Interface Interface (by ref)", PlayerPawn:Fort Pawn Object Reference, SeatIndex:int) {is=title K2Node_FunctionEntry}
   > Branch(Condition="Equal (Integer)"(A=@n1.SeatIndex))
   > RefreshWarningState()
  ]}
 OnVehicleDisabledChanged {nodes [
   OnVehicleDisabledChanged!() -> (then, bIsDisabled:bool=~) {is=title K2Node_FunctionEntry}
   > RefreshWarningState()
  ]}
 OnViewTargetChanged {nodes [
   n1 = OnViewTargetChanged!() -> (then, PlayerController:Fort Player Controller Object Reference, OldViewTarget:Actor Object Reference, NewViewTarget:Actor Object Reference) {is=title K2Node_FunctionEntry}
   > SetViewTarget(NewViewTarget=@n1.NewViewTarget)
  ]}
 RefreshWarningState {nodes [
   RefreshWarningState() {is=title K2Node_FunctionEntry}
   n1 = Get BoundVehicle() -> (BoundVehicle:Fort Athena Vehicle Object Reference)
   > n2 = Is Valid(InputObject=@n1)
   > Set IsVehicleDisabled(IsVehicleDisabled:bool=IsDisabled!(self:Fort Vehicle Interface Interface=@n1) -> (ReturnValue:bool)) -> (Output_Get:bool)
   > n3 = PassengerIsOnDriverTeam!(
    execute
    Passenger:Fort Player Pawn Object Reference=Get CurrentViewTarget() -> (CurrentViewTarget:Fort Player Pawn Object Reference)
    DefaultTeamResult:bool=true
   ) -> (then, ReturnValue:bool)
   > n4 = Set IsDriverOnSameTeam(IsDriverOnSameTeam:bool=@n3) -> (Output_Get:bool)
   > n5 = Branch(Condition=OR Boolean(A=NOT Boolean(A=@n4), B=Get IsVehicleDisabled() -> (IsVehicleDisabled:bool)))
   > SetVisibility(InVisibility=SelfHitTestInvisible)
   > "SetText (Text)"(
    self=Get ReticleStatusText() -> (ReticleStatusText:Common Text Object Reference)
    InText=SelectText(
     A="LOCTABLE(\"/VehicleModsUI/Data/ST_VehicleMods_ErrorMessages\", \"VehicleDisabled\")"
     B="LOCTABLE(\"&ST_VehicleMods_ErrorMessages\", \"EnemyPlayerOwned\")"
     bPickA=Get IsVehicleDisabled()
    )
   )
   SetVisibility(execute=[@n5.else, @n2.Is Not Valid], InVisibility=Hidden)
  ]}
 SetViewTarget {nodes [
   n1 = SetViewTarget!() -> (then, NewViewTarget:Actor Object Reference) {is=title K2Node_FunctionEntry}
   > n2 = Cast To FortPlayerPawn(Object=@n1)
   > n3 = Set CurrentViewTarget(CurrentViewTarget:Fort Player Pawn Object Reference=@n2) -> (Output_Get:Fort Player Pawn Object Reference)
   > n4 = Cast To FortAthenaVehicle!(execute, Object=GetVehicleActor!(self:Fort Player Pawn Object Reference=@n3) -> (ReturnValue:Actor Object Reference)) -> (then, CastFailed, AsFort Athena Vehicle:Fort Athena Vehicle Object Reference) {is=title K2Node_DynamicCast}
   > n5 = Set NewBoundVehicle(NewBoundVehicle:Fort Athena Vehicle Object Reference=@n4) -> (Output_Get:Fort Athena Vehicle Object Reference)
   > Branch(Condition="Not Equal (Object)"(A:Object Reference=@n5, B:Object Reference=Get BoundVehicle()) -> (ReturnValue:bool=~) {is=title K2Node_PromotableOperator})
   > UnbindVehicleEvents()
   > n6 = Set BoundVehicle(BoundVehicle:Fort Athena Vehicle Object Reference=Get NewBoundVehicle() -> (NewBoundVehicle:Fort Athena Vehicle Object Reference)) -> (Output_Get:Fort Athena Vehicle Object Reference)
   > Bind Event to On Disabled Changed Delegate(self:Fort Athena Vehicle Object Reference=@n6, Delegate=Create Event()) {is=title K2Node_AddDelegate}
   n7 = Create Event()
   > Bind Event to On Pawn Entered Vehicle Seat(self:Fort Athena Vehicle Object Reference=@n6, Delegate=@n7) {is=title K2Node_AddDelegate}
   > n8 = Bind Event to On Pawn Exited Vehicle Seat(self:Fort Athena Vehicle Object Reference=@n6, Delegate=@n7) {is=title K2Node_AddDelegate}
   RefreshWarningState(execute=[@n8.then, @n9.then]) {is=title K2Node_CallFunction}
   Set CurrentViewTarget(execute=[@n4.CastFailed, @n2.CastFailed])
   > n9 = UnbindVehicleEvents()
  ]}
 UnbindVehicleEvents {nodes [
   UnbindVehicleEvents() {is=title K2Node_FunctionEntry}
   n1 = Get BoundVehicle()
   > Is Valid(InputObject=@n1)
   > Unbind Event from On Disabled Changed Delegate(self:Fort Athena Vehicle Object Reference=@n1, Delegate=Create Event()) {is=title K2Node_RemoveDelegate}
   n2 = Create Event()
   > Unbind Event from On Pawn Entered Vehicle Seat(self:Fort Athena Vehicle Object Reference=@n1, Delegate=@n2) {is=title K2Node_RemoveDelegate}
   > Unbind Event from On Pawn Exited Vehicle Seat(self:Fort Athena Vehicle Object Reference=@n1, Delegate=@n2) {is=title K2Node_RemoveDelegate}
   > Set BoundVehicle()
  ]}
]
variables [
 BoundVehicle {type=FortAthenaVehicle, default=null}
 CurrentViewTarget {type=FortPlayerPawn}
]
widgets {
 info {inheritedWidgetCount=0, parentClass=&UserWidget, rootWidgetClass=&CommonVisualAttachment, widgetCount=2}
 widgets [
  CommonVisualAttachment_1 {
   widgetClassPath=&CommonVisualAttachment
   widget=CommonVisualAttachment_1
   properties {renderTransform {angle=0, scale=<1 1>, shear=<0 0>, translation=<195 0>}}
  }
  ReticleStatusText {
   widgetClassPath=&CommonTextBlock
   bIsVariable=true
   parent=CommonVisualAttachment_1
   widget=ReticleStatusText
   properties {
    justification=Center
    mobileFontSizeMultiplier=1.2
    text="{Reticle Status}"
    font {
     bForceMonospaced=false
     bMaterialIsStencil=false
     fontMaterial=null
     fontObject=&BurbankSmall
     letterSpacing=0
     monospacedWidth=1
     size=18
     skewAmount=0
     typefaceFontName=Bold
     outlineSettings {
      bApplyOutlineToDropShadows=true
      bMiteredCorners=false
      bSeparateFillAlpha=true
      outlineMaterial=&M-FontShadow
      outlineSize=3
      outlineColor=<rgba 0 0 0 0.5>
     }
    }
    shadowOffset=<0 0>
   }
   slot {class=SizeBoxSlot}
  }
 ]
}
